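# Niagara system for the CoconutShell emote's plant glow, with one LightAndGlow emitter whose only renderer is a light renderer.
# A burst spawns two long-lived particles in separate spawn groups, each following a skeletal mesh bone or socket indexed by its group.
# Curves fade the pale blue light in and out, with brightness scaled by time of day, strongest at night.
# Sprite sizing and owner-axis material parameters are also written, though no sprite renderer is present.
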NiagaraSystem /BRCosmetics/Animation/Game/MainPlayer/Emotes/CoconutShell/FX/NS_Emote_CoconutShell_PlantGlow
dependencies {
 usedDataInterfaces [&NiagaraDataInterfaceCurve, &NiagaraDataInterfaceSkeletalMesh]
 usedDynamicInputs []
 usedModules [
  &self:SystemSpawnScript.SystemScriptSource.SystemScriptGraph.NiagaraNodeAssignment_0.SetVariables_#1
  &SystemState
 ]
}
emitters {LightAndGlow {
  emitterUpdateScript [
   &EmitterState(
    Inactive Response=::NewEnumerator1
    Life Cycle Mode=::NewEnumerator1
    Loop Behavior=::NewEnumerator1
    Loop Duration Mode=::NewEnumerator1
    UseLoopDelay=false
   )
   &SpawnBurst_Instantaneous(Spawn Count=2, Use Spawn Probability=false, UseLoopCountLimit=false)
  ]
  particleSpawnScript [
   &InitializeParticle(
    Color=<rgba 0.2127 0.676 0.9323 0.5>
    Lifetime=16
    Sprite Size Mode=::NewEnumerator3
    Uniform Sprite Size=300
    UsePositionOffset=true
    Write Material Random=true
   )
   &FN_SetSpawnGroup(Number of Spawn Groups=2)
  ]
  particleUpdateScript [
   &ParticleState()
   &SkeletalMeshLocation(
    Bone / Socket Index=&SpawnGroup()
    Bone Sampling Mode=::NewEnumerator4
    MaskSpawnGroup=false
    Skeletal Mesh=$System.Skeletal Mesh
   )
   &ScaleColor(
    Scale Alpha=&FloatFromCurve(FloatCurve=DataInterface() {propertyValues {Curve {keys [
         {interpMode=RCIM_Cubic, tangentMode=RCTM_User, time=0, value=0}
         {arriveTangent=0.9211, interpMode=RCIM_Cubic, leaveTangent=0.9211, time=0.05, value=0.8}
         {arriveTangent=-0.8421, interpMode=RCIM_Cubic, leaveTangent=-0.8421, time=0.95, value=0.875}
         {tangentMode=RCTM_User, time=1, value=0}
        ]}}})
    ScaleA=true
    ScaleRGB=false
   )
   &DynamicMaterialParameters(
    Index 0 Param 1=&MakeFloatFromVector(Vector=$Engine.Owner.SystemYAxis, Channel=::NewEnumerator0)
    Index 0 Param 2=&MakeFloatFromVector(Vector=$Engine.Owner.SystemYAxis, Channel=::NewEnumerator1)
    Index 0 Param 3=&MakeFloatFromVector(Vector=$Engine.Owner.SystemYAxis, Channel=::NewEnumerator2)
    Index 0 Param 4=150
    Module.Param0WriteEnabled=true
    Module.Param1WriteEnabled=true
    Module.Param2WriteEnabled=true
    Module.Param3WriteEnabled=true
   )
   &FN_Light_Attributes(
    Color=<rgba 0.4892 0.8273 1 1>
    Create Color Attribute=true
    Create Radius Attribute=true
    Radius=&FloatFromCurve(
     FloatCurve=DataInterface() {propertyValues {Curve {keys [{interpMode=RCIM_Cubic, tangentMode=RCTM_User, time=0.05, value=1}]}}}
     Scale Curve=75
    )
    Scale Color=&FloatFromCurve(
     FloatCurve=DataInterface() {propertyValues {Curve {keys [
         {arriveTangent=27.2269, interpMode=RCIM_Cubic, leaveTangent=27.2269, tangentMode=RCTM_User, time=0, value=0}
         {interpMode=RCIM_Cubic, tangentMode=RCTM_User, time=0.05, value=1}
         {arriveTangent=-0.0068, interpMode=RCIM_Cubic, leaveTangent=-0.0068, tangentMode=RCTM_User, time=0.95, value=1}
         {arriveTangent=-35.012, tangentMode=RCTM_User, time=1, value=0}
        ]}}}
     Scale Curve=&Multiply_Float(A=2, B=&FNTimeOfDayScalar(Night Scale=1.5, Morning Scale=0.4, Day Scale=0.2, Evening Scale=0.4))
    )
   )
  ]
  rendererClasses [&NiagaraLightRendererProperties]
  renderers [{rendererClass=&NiagaraLightRendererProperties, rendererIndex=0}]
 }}
summary {
 systemName=NS_Emote_CoconutShell_PlantGlow
 emitters [{emitterName=LightAndGlow, rendererClasses [&NiagaraLightRendererProperties]}]
 userVariables []
}
