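# Niagara effect for the CoconutShell emote that releases seven pale-blue wood sprites, drawn as mesh plus sprite, over a roughly 14-second cycle.
# Particles burst outward from skeletal mesh bone samples, then blend into a staggered, bobbing orbit around an elevated center.
# Randomized per-particle looping phases synchronize bob, orbit rate and a material WPO parameter.
# Smoothed velocity, fade and scale curves, time-of-day alpha and light attributes complete the look.
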NiagaraSystem /BRCosmetics/Animation/Game/MainPlayer/Emotes/CoconutShell/FX/NS_Emote_CoconutShell_WoodSprite
dependencies {
 usedDataInterfaces [
  &NiagaraDataInterfaceMeshRendererInfo
  &NiagaraDataInterfaceCurve
  &NiagaraDataInterfaceSkeletalMesh
 ]
 usedDynamicInputs []
 usedModules [
  &self:SystemSpawnScript.SystemScriptSource.SystemScriptGraph.NiagaraNodeAssignment_0.SetVariables_#1
  &SystemState
 ]
}
emitters {Wood_Sprites {
  emitterUpdateScript [
   &EmitterState(Inactive Response=::NewEnumerator1, Life Cycle Mode=::NewEnumerator1, Loop Behavior=::NewEnumerator1, Loop Duration=14, UseLoopDelay=false)
   &SpawnBurst_Instantaneous(Spawn Count=7, Use Spawn Probability=false, UseLoopCountLimit=false)
  ]
  particleSpawnScript [
   &InitializeParticle(
    Color=<rgba 0.3698 0.7374 1 0.05>
    Lifetime Max=13.75
    Lifetime Min=13.5
    Lifetime Mode=::NewEnumerator1
    Mesh Scale=&Multiply_VectorByFloat(Vector=<1 1.25 1>, Float=0.6)
    Mesh Scale Mode=::NewEnumerator4
    Position Offset=<0 0 150>
    Sprite Size Mode=::NewEnumerator3
    Uniform Sprite Size=20
    UsePositionOffset=true
    Write Material Random=true
   )
   &InitialMeshOrientation(Mesh Axis To Orient=<1 0 0.5>, Mesh Orientation Mode=::NewEnumerator2, Use Rotation Vector=false)
   &PartitionParticles(Number Of Partitions=1, Particles Per Partition=7, Partition Mode:ENiagara_PartitionMode=::NewEnumerator1)
   /SetVariables() {isSetParameters=true}
   /SetVariables(
    Particles.LoopCountViaFrac:NiagaraFloat=&RandomRangeFloat(Minimum=4, Maximum=5)
    Particles.TimingOffset:NiagaraFloat=&RandomRangeFloat(Minimum=0, Maximum=1)
   ) {isSetParameters=true}
   &SkeletalMeshLocation(Bone Sampling Mode=::NewEnumerator4, MaskSpawnGroup=false, Orientation Sampling=::NewEnumerator1, Skeletal Mesh=$System.Skeletal Mesh)
   &AddVelocity(
    Cone Angle=30
    Cone Axis=&RotateVectorByQuaternion(Vector To Rotate=<-1 0 1>, Quaternion=$Particles.SkeletalMeshLocation.SampledRotation)
    Module.UseConeFalloff=true
    Module.UseVelDistribution=true
    Override Seed=false
    Velocity Mode=::NewEnumerator2
    Velocity Speed=1
   )
   /SetVariables(Particles.StartingPosition:NiagaraFloat=&Add_Float(A=0.2, B=&Multiply_Float(A=0.3, B=$Particles.PartitionParticles.Normalized Index In Partition))) {isSetParameters=true}
  ]
  particleUpdateScript [
   &ParticleState()
   /SetVariables(Particles.PositionBurst:NiagaraPosition=&AddVectorToPosition(
     Position=&Lerp_Vector(
      A=$Particles.SkeletalMeshLocation.SampledPosition
      B=<0 0 200>
      Alpha=&FloatFromCurve(FloatCurve=DataInterface() {propertyValues {Curve {keys [<at 0 0>,<at 0.8 0>,<at 0.9 1>,<at 1 1>]}}})
     )
     Vector=&Multiply_VectorByFloat(
      Vector=&NormalizeVector(
       Vector To Normalize=&Add_Vector(
        A=&RotateVectorByQuaternion(Vector To Rotate=<-1 -0.1 1>, Quaternion=$Particles.SkeletalMeshLocation.SampledRotation)
        B=&RandomVector(Vector Scale=0.35)
       )
       Safe Normalize=true
       Safe Fallback Threshold=0
       Safe Fallback Vector=<0 0 0.001>
      )
      Float=&FloatFromCurve(
       FloatCurve=DataInterface() {propertyValues {Curve {keys [
           {arriveTangent=17.5323, interpMode=RCIM_Cubic, leaveTangent=17.5323, tangentMode=RCTM_User, time=0, value=0}
           {interpMode=RCIM_Cubic, tangentMode=RCTM_User, time=0.15, value=1}
           {arriveTangent=0.0052, interpMode=RCIM_Cubic, leaveTangent=0.0052, tangentMode=RCTM_User, time=0.8, value=1}
           <at 0.95 0 RCIM_Cubic>
          ]}}}
       Scale Curve=&RandomRangeFloat(Minimum=80, Maximum=120)
      )
     )
    )) {isSetParameters=true}
   &RotateAroundPoint(
    Radius:NiagaraFloat=&FloatFromCurve(
     FloatCurve=DataInterface() {propertyValues {Curve {keys [
         {arriveTangent=4.1769, interpMode=RCIM_Cubic, leaveTangent=4.1769, tangentMode=RCTM_User, time=0, value=0.7}
         {interpMode=RCIM_Cubic, tangentMode=RCTM_User, time=0.2, value=1}
         {arriveTangent=-0.0072, interpMode=RCIM_Cubic, leaveTangent=-0.0072, tangentMode=RCTM_User, time=0.4, value=0.8}
         {arriveTangent=-0.1667, interpMode=RCIM_Cubic, leaveTangent=-0.1667, time=0.7, value=0.875}
         {tangentMode=RCTM_User, time=1, value=0.7}
        ]}}}
     Scale Curve=&RandomRangeFloat(Minimum=45, Maximum=55)
    )
    Rotation Center:NiagaraPosition=&AddVectorToPosition(
     Position=&RandomRangeVector(Minimum=<0 15 130>, Maximum=<0 20 180>)
     Vector=&MakeVector(
      X=0
      Y=0
      Z=&Add_Float(
       A=&FloatFromCurve(
        FloatCurve=DataInterface() {propertyValues {Curve {keys [
            {arriveTangent=0.0242, interpMode=RCIM_Cubic, leaveTangent=0.0242, tangentMode=RCTM_User, time=0, value=0}
            {arriveTangent=-0.0026, interpMode=RCIM_Cubic, leaveTangent=-0.0026, tangentMode=RCTM_User, time=0.25, value=1}
            {arriveTangent=0.0242, tangentMode=RCTM_User, time=1, value=0}
           ]}}}
        CurveIndex=&Frac_Float(Float=&Add_Float(A=&Multiply_Float(A=$Particles.NormalizedAge, B=$Particles.LoopCountViaFrac), B=$Particles.TimingOffset))
        Scale Curve=&RandomRangeFloat(Minimum=7.5, Maximum=15)
       )
       B=&FloatFromCurve(
        FloatCurve=DataInterface() {propertyValues {Curve {keys [
            <at 0.1 0 RCIM_Cubic>
            <at 0.4 -0.15 RCIM_Cubic>
            {arriveTangent=0.1301, interpMode=RCIM_Cubic, leaveTangent=0.0715, tangentMode=RCTM_Break, time=0.7, value=0}
            {arriveTangent=-0.0262, interpMode=RCIM_Cubic, leaveTangent=-0.0262, tangentMode=RCTM_User, time=1, value=1.5}
           ]}}}
        Scale Curve=&RandomRangeFloat(Minimum=30, Maximum=60)
       )
      )
     )
    )
    Rotation Phase:NiagaraFloat=&Add_Float(
     A=$Particles.StartingPosition
     B=&FloatFromCurve(
      FloatCurve=DataInterface() {propertyValues {Curve {keys [
          {arriveTangent=0.0535, interpMode=RCIM_Cubic, leaveTangent=0.0535, tangentMode=RCTM_User, time=0, value=0}
          {interpMode=RCIM_Cubic, tangentMode=RCTM_User, time=0.5, value=1}
         ]}}}
      Scale Curve=&Multiply_Float(
       A=&Frac_Float(Float=&Subtract_Float(A=$Particles.PartitionParticles.Normalized Index In Partition, B=$Particles.StartingPosition))
       B=1
      )
     )
    )
    Rotation Rate:NiagaraFloat=&FloatFromCurve(
     FloatCurve=DataInterface() {propertyValues {Curve {keys [
         {interpMode=RCIM_Cubic, tangentMode=RCTM_User, time=0, value=0.8}
         {arriveTangent=0.0044, interpMode=RCIM_Cubic, leaveTangent=0.0044, tangentMode=RCTM_User, time=0.3, value=1.2}
         {tangentMode=RCTM_User, time=0.5, value=0.8}
        ]}}}
     CurveIndex=&Frac_Float(Float=&Add_Float(A=&Multiply_Float(A=$Particles.NormalizedAge, B=$Particles.LoopCountViaFrac), B=$Particles.TimingOffset))
     Scale Curve=&RandomRangeFloat(Minimum=-0.1, Maximum=-0.075)
    )
    Set Particle Position=false
   )
   &ScaleSpriteSize(Uniform Curve Sprite Scale=DataInterface() {propertyValues {Curve {keys [
        {arriveTangent=20.8093, interpMode=RCIM_Cubic, leaveTangent=20.8093, tangentMode=RCTM_User, time=0, value=0.5}
        {interpMode=RCIM_Cubic, tangentMode=RCTM_User, time=0.05, value=1}
        {arriveTangent=-0.0044, interpMode=RCIM_Cubic, leaveTangent=-0.0044, tangentMode=RCTM_User, time=0.75, value=1}
        {arriveTangent=-0.2128, tangentMode=RCTM_User, time=0.95, value=0}
       ]}}})
   &ScaleMeshSize(Scale Factor=&VectorFromFloat(Value=&FloatFromCurve(FloatCurve=DataInterface() {propertyValues {Curve {keys [
          {arriveTangent=13.6796, interpMode=RCIM_Cubic, leaveTangent=13.6796, tangentMode=RCTM_User, time=0, value=0}
          {interpMode=RCIM_Cubic, tangentMode=RCTM_User, time=0.12, value=1}
          {arriveTangent=-0.0066, interpMode=RCIM_Cubic, leaveTangent=-0.0066, tangentMode=RCTM_User, time=0.7, value=1}
          {arriveTangent=-0.0051, tangentMode=RCTM_User, time=0.9, value=0}
         ]}}})))
   /SetVariables(Particles.Position=&Lerp_Vector(
     A=Linked() {linkedVariable {name=Particles.PositionBurst, type=NiagaraPosition}}
     B=$Output.RotateAroundPoint.OutputPosition
     Alpha=&FloatFromCurve(FloatCurve=DataInterface() {propertyValues {Curve {keys [
          {arriveTangent=3.0862, interpMode=RCIM_Cubic, leaveTangent=3.0862, tangentMode=RCTM_User, time=0, value=0}
          {interpMode=RCIM_Cubic, tangentMode=RCTM_User, time=0.25, value=1}
         ]}}})
    )) {isSetParameters=true}
   &CalculateAccurateVelocity()
   &InterpolateOverTime(
    Initial Vector Value:Vector3f=$Particles.Initial.Velocity
    Rate Of Change=3
    Target Vector Value:Vector3f=&NormalizeVector(Vector To Normalize=$Particles.Velocity, Safe Normalize=true, Safe Fallback Threshold=0, Safe Fallback Vector=<0 0 0.001>)
    Variable Type=::NewEnumerator2
   )
   /SetVariables(Particles.Velocity=$Particles.InterpolateOverTime.InterpolateOverTime_V3_.Moving Average) {isSetParameters=true}
   &ScaleColor(
    Scale Alpha=&FloatFromCurve(
     FloatCurve=DataInterface() {propertyValues {Curve {keys [
         {arriveTangent=6.9613, interpMode=RCIM_Cubic, leaveTangent=6.9613, tangentMode=RCTM_User, time=0, value=0}
         {interpMode=RCIM_Cubic, tangentMode=RCTM_User, time=0.3, value=1}
         {arriveTangent=-0.0044, interpMode=RCIM_Cubic, leaveTangent=-0.0044, tangentMode=RCTM_User, time=0.7, value=1}
         {arriveTangent=0.0005, tangentMode=RCTM_User, time=1, value=0}
        ]}}}
     Scale Curve=&Multiply_Float(
      A=0.75
      B=&FNTimeOfDayScalar(Night Scale=1, Morning Scale=0.75, Day Scale=0.5, Evening Scale=0.75)
     )
    )
    ScaleA=true
    ScaleRGB=true
   )
   /SetVariables(Particles.WPO Offset:NiagaraFloat=&FloatFromCurve(
     FloatCurve=DataInterface() {propertyValues {Curve {keys [
         {arriveTangent=-0.7811, interpMode=RCIM_Cubic, leaveTangent=-0.7811, tangentMode=RCTM_User, time=0, value=0.3}
         {arriveTangent=-0.9606, interpMode=RCIM_Cubic, leaveTangent=-0.9606, tangentMode=RCTM_User, time=0.2, value=-0.15}
         {arriveTangent=-0.6934, interpMode=RCIM_Cubic, leaveTangent=-0.6934, tangentMode=RCTM_User, time=1, value=-0.7}
        ]}}}
     CurveIndex=&Frac_Float(Float=&Add_Float(
       A=&Multiply_Float(A=$Particles.NormalizedAge, B=$Particles.LoopCountViaFrac)
       B=&Add_Float(A=$Particles.TimingOffset, B=0.1)
      ))
    )) {isSetParameters=true}
   &DynamicMaterialParameters(
    Index 0 Param 1=$Particles.WPO Offset
    Module.Index0Param0Write=true
    Module.Index0Param1Write=true
    Module.Index0Param2Write=true
    Module.Index0Param3Write=true
   )
   &FN_Light_Attributes(
    Color=<rgba 0.4892 0.8273 1 1>
    Create Color Attribute=true
    Create Radius Attribute=true
    Radius=&FloatFromCurve(
     FloatCurve=DataInterface() {propertyValues {Curve {keys [
         {arriveTangent=6.9613, interpMode=RCIM_Cubic, leaveTangent=6.9613, tangentMode=RCTM_User, time=0, value=0}
         {interpMode=RCIM_Cubic, tangentMode=RCTM_User, time=0.1, value=1}
         {arriveTangent=-0.0044, interpMode=RCIM_Cubic, leaveTangent=-0.0044, tangentMode=RCTM_User, time=0.8, value=1}
         {arriveTangent=-6.7325, tangentMode=RCTM_User, time=1, value=0}
        ]}}}
     Scale Curve=30
    )
    Scale Color=&FloatFromCurve(
     FloatCurve=DataInterface() {propertyValues {Curve {keys [
         {interpMode=RCIM_Cubic, tangentMode=RCTM_User, time=0, value=0}
         {interpMode=RCIM_Cubic, tangentMode=RCTM_User, time=0.1, value=1}
         {arriveTangent=-0.0068, interpMode=RCIM_Cubic, leaveTangent=-0.0068, tangentMode=RCTM_User, time=0.9, value=1}
         {tangentMode=RCTM_User, time=1, value=0}
        ]}}}
     Scale Curve=5
    )
   )
  ]
  rendererClasses [&NiagaraMeshRendererProperties, &NiagaraSpriteRendererProperties]
  renderers [
   {rendererClass=&NiagaraMeshRendererProperties, rendererIndex=0}
   {rendererClass=&NiagaraSpriteRendererProperties, rendererIndex=1}
  ]
 }}
summary {
 systemName=NS_Emote_CoconutShell_WoodSprite
 emitters [{emitterName=Wood_Sprites, rendererClasses [&NiagaraMeshRendererProperties, &NiagaraSpriteRendererProperties]}]
 userVariables []
}
